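# Imported Default Lit material named for the tilt/roll control of a camera rig in the WildEstate layout content.
# The graph holds only two constant colours, a grey and a pure red, with no textures, parameters or math.
# No links from either colour to a material output appear, so their roles in the shading are not evident.
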Material /WildEstate/Layout/CamRigCTRLS/M_3_TiltRoll
graph {
 properties {assetImportData="<self>:InterchangeAssetImportData_0", usedShadingModels=Default Lit}
 nodes [
  Constant3Vector() {constant=<rgba 0.2159 0.2159 0.2159 1>}
  Constant3Vector() {constant=<rgba 1 0 0 1>}
 ]
}
